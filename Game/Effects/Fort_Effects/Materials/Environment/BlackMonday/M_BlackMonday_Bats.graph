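# Translucent unlit, two-sided BlackMonday effect material for animated bat meshes and particles, tinted by particle color.
# World position offset sums a sine wing flap shaped by vertex colors and dynamic parameters, vector-noise drift, and a normal push that grows with camera distance to enlarge distant bats.
# Opacity uses a bat mask, with a static switch choosing smoke-blended erosion or contrast-adjusted alpha, depth-faded except on mobile.
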Material /Game/Effects/Fort_Effects/Materials/Environment/BlackMonday/M_BlackMonday_Bats
graph {
 properties {
  bUsedWithMeshDeformer=true
  bUsedWithMeshParticles=true
  bUsedWithNiagaraMeshParticles=true
  bUsedWithSkeletalMesh=true
  blendMode=BLEND_Translucent
  pixelDepthOffsetMode=PDOM_Legacy
  shadingModel=MSM_Unlit
  twoSided=true
  usedShadingModels=Unlit
 }
 nodes [
  n1 = ComponentMask(None=TextureCoordinate()) {r=true}
  n2 = VertexColor()
  n3 = Multiply(
   A=Sign(None=Subtract(A=@n1) {constB=0.5})
   B=LinearInterpolate(
    A=ScalarParameter() {defaultValue=-0.389, parameterName=InwardVector}
    B=ScalarParameter() {defaultValue=-1.0138, parameterName=InwardVector_1}
    Alpha=@n2.G
   )
  )
  n4 = ScalarParameter() {defaultValue=0.134, parameterName=ForwardVector}
  n5 = AppendVector(
   A=StaticSwitchParameter(True=AppendVector(A=@n3, B=@n4), False=AppendVector(A=@n4, B=@n3)) {parameterName=SwizzleXY}
   B=LinearInterpolate(
    A=ScalarParameter() {defaultValue=0.5823, parameterName=UPVector}
    B=ScalarParameter() {defaultValue=1.4967, parameterName=UPVector_1}
    Alpha=@n2.B
   )
  )
  n6 = Time()
  n7 = VertexColor()
  n8 = ScalarParameter() {defaultValue=0.2, parameterName=WingOffset}
  n9 = ScalarParameter() {defaultValue=2, parameterName=SpeedWings}
  n10 = VertexColor()
  n11 = DynamicParameter() {defaultValue=<rgba 1 1 0.3 1>, paramNames [WingAnimTimeOffset, WingAnimSpeedMultiplier, ErodeAlpha, SmokeLerp]}
  n12 = DynamicParameter() {defaultValue=<rgba 1 1 0.3 1>, paramNames [WingAnimTimeOffset, WingAnimSpeedMultiplier, ErodeAlpha, SmokeLerp]}
  n13 = Add(A=@n12.WingAnimSpeedMultiplier, B=VertexColor().A)
  n14 = Multiply(
   A=Add(A=Add(A=@n6, B=@n7.A), B=Multiply(A=@n2.G, B=@n8) {constB=0.5})
   B=Multiply(A=Multiply(A=@n9, B=@n10.A), B=Add(A=@n11.WingAnimTimeOffset, B=@n13))
  )
  n15 = Multiply(
   A=Multiply(A=Transform(None=@n5) {transformSourceType=TRANSFORMSOURCE_Local}, B=ScalarParameter() {defaultValue=15, parameterName=WindMovementDistance})
   B=Add(A=Sine(None=@n14), B=ScalarParameter() {defaultValue=0.7853, parameterName=WindMovementBias})
  )
  n16 = Add(A=@n2.R, B=ScalarParameter() {defaultValue=0.2286, parameterName=BodyMovement})
  n17 = VertexNormalWS()
  n18 = Multiply(A=@n17, B=ScalarParameter() {defaultValue=25, parameterName=SizeScale})
  n19 = WorldPosition()
  n20 = CameraPositionWS()
  n21 = Clamp(None=Divide(A=Distance(A=@n19.XYZ, B=@n20)) {constB=15000})
  n22 = ScalarParameter() {defaultValue=50, parameterName=CurIntensity}
  n23 = ScalarParameter() {defaultValue=100, parameterName=CurIntensityHigh}
  n24 = VertexColor()
  n25 = WorldPosition()
  n26 = ActorPositionWS()
  n27 = Divide(
   A=Subtract(A=Subtract(A=@n25.XYZ, B=@n26), B=ScalarParameter() {defaultValue=1, parameterName=CurlSpeed})
   B=LinearInterpolate(
    A=ScalarParameter() {defaultValue=1000, parameterName=CurlScale}
    B=ScalarParameter() {defaultValue=500, parameterName=CurlScale_Low}
    Alpha=@n24.A
   )
  )
  Add(
   A=Add(A=Multiply(A=@n15, B=@n16), B=LinearInterpolate(A=@n17, B=@n18, Alpha=@n21))
   B=Multiply(A=LinearInterpolate(A=@n22, B=@n23, Alpha=@n24.A), B=VectorNoise(World Position=@n27) {noiseFunction=VNF_VectorALU})
  )
  MaterialFunctionCall(Emissive=ParticleColor().RGB) {materialFunction=&MF_FX_EmissiveAdjustments}
  n28 = TextureSample() {samplerType=SAMPLERTYPE_Masks, texture=/Game/Gadgets/Assets/VinderTech_GliderChute/Glider_Vampire/Textures/T_Glider_Bat_M}
  n29 = Multiply(
   A=@n28.G
   B=TextureSample(UVs=TextureCoordinate() {uTiling=3, vTiling=3}) {samplerType=SAMPLERTYPE_Grayscale, texture=&T_Smoke_Trail_Soft}.G
  )
  n30 = DynamicParameter() {defaultValue=<rgba 1 1 0.3 1>, paramNames [WingAnimTimeOffset, WingAnimSpeedMultiplier, ErodeAlpha, SmokeLerp]}
  n31 = DynamicParameter() {defaultValue=<rgba 1 1 0.3 1>, paramNames [WingAnimTimeOffset, WingAnimSpeedMultiplier, ErodeAlpha, SmokeLerp]}
  n32 = Multiply(A=@n28.R, B=@n28.R)
  n33 = ParticleColor()
  n34 = Clamp(None=Power(Base=Multiply(A=@n32, B=@n33.A), Exp=ScalarParameter() {defaultValue=0.6555, parameterName=Contrast}))
  n35 = ScalarParameter() {defaultValue=40, parameterName=DepthFade}
  StaticSwitchParameter(
   True=MaterialFunctionCall!(Dynamic Param=LinearInterpolate(A=@n28.G, B=@n29, Alpha=@n30.SmokeLerp), Alpha=@n31.ErodeAlpha) {materialFunction=/Game/Effects/Fort_Effects/Materials/_Functions/Erode_Alpha}.Result
   False=ShadingPathSwitch(Default=DepthFade(Opacity=@n34, FadeDistance=@n35), Mobile=@n34) {desc=Migrated from FeatureLevelSwitch}
  ) {parameterName=AddAlphaSmokeErode}
 ]
}
