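# Translucent, two-sided unlit material for animated BlackMonday environment fire plumes on spline-thickened strips.
# Opacity multiplies panning, normal-distorted smoke and noise, offset by object position and vertex color, with edge masks, depth fade and distance fading.
# Emissive blends two hue-shifted colors along V with HDR remapping.
# World position offset combines distance-scaled SplineThicken width with noise waves, while low-quality, mobile and forward paths simplify these effects.
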Material /Game/Effects/Fort_Effects/Materials/Environment/BlackMonday/M_BlackMonday_FirePlumes
graph {
 properties {blendMode=BLEND_Translucent, pixelDepthOffsetMode=PDOM_Legacy, shadingModel=MSM_Unlit, twoSided=true, usedShadingModels=Unlit}
 nodes [
  Frac()
  n1 = TextureCoordinate()
  n2 = ComponentMask(None=@n1) {g=true}
  n3 = ComponentMask(None=@n1) {r=true}
  n4 = Multiply(B=Multiply(A=@n3, B=OneMinus(None=@n3))) {constA=4}
  n5 = Multiply(B=Multiply(A=Power(Base=@n2), B=Power(Base=@n4) {constExponent=4})) {constA=4}
  n6 = Multiply(A=@n5, B=Clamp(None=@n5))
  n7 = ComponentMask(None=Divide(A=ObjectPositionWS(), B=ScalarParameter() {parameterName=WorldPosUVOffset})) {g=true, r=true}
  n8 = ComponentMask(None=VectorParameter() {parameterName=Tile_02_Mult, defaultValue=<rgba 0.5 0.2 0 1>}.RGB) {g=true, r=true}
  n9 = Multiply(A=TextureCoordinate(), B=@n8)
  n10 = VertexColor()
  n11 = Panner(
   Coordinate=Add(A=@n7, B=Add(A=@n9, B=@n10.G))
   Speed=Multiply(A=VertexColor().G, B=ScalarParameter() {defaultValue=0.25, parameterName=Speed_01})
  ) {speedY=1}
  n12 = ScalarParameter() {defaultValue=0.25, parameterName="Rotation Angle (0-1) (S)"}
  n13 = TextureCoordinate()
  n14 = ComponentMask(None=VectorParameter() {parameterName=Tile_01_Mult, defaultValue=<rgba 0.25 1 0 1>}.RGB) {g=true, r=true}
  n15 = Add(A=@n10.B, B=Add(A=@n7, B=Multiply(A=@n13, B=@n14)))
  n16 = Add(
   A=TextureSample(UVs=MaterialFunctionCall(UVs=@n11, "Rotation Angle (0-1)"=@n12) {materialFunction=&CustomRotator}.Rotated Values) {samplerType=SAMPLERTYPE_Masks, texture=/Game/Weapons/WeaponSkins/Wraps/Textures/Mash/Mash_Noise}.RGB
   B=TextureSample(UVs=Panner(Coordinate=@n15) {speedY=0.5}) {samplerType=SAMPLERTYPE_Masks, texture=&Mash_Noise}.RGB
  )
  n17 = TextureCoordinate() {vTiling=0.7}
  n18 = ComponentMask(None=@n17) {r=true}
  n19 = VertexColor()
  n20 = AppendVector(A=Multiply(A=Add(A=@n18, B=@n19.G), B=@n19.R), B=ComponentMask(None=@n17) {g=true})
  n21 = ScalarParameter() {defaultValue=0.75, parameterName=TimeScale_Low}
  n22 = ScalarParameter() {defaultValue=1.25, parameterName=TimeScale_High}
  n23 = VertexColor()
  n24 = Add(A=Multiply(A=Time(), B=LinearInterpolate(A=@n21, B=@n22, Alpha=@n23.G)) {constB=0.25}, B=Multiply(A=VertexColor().G) {constB=2})
  n25 = Panner(Coordinate=TextureCoordinate() {uTiling=0.25, vTiling=0.25}, Time=@n24) {speedX=0.1, speedY=0.25}
  n26 = Multiply(
   A=ComponentMask(None=TextureSample(UVs=@n25) {samplerType=SAMPLERTYPE_Normal, texture=/Game/Effects/Fort_Effects/Gameplay/SmokeDustExplosions/Textures/T_Explosion_NM_01}.RGB) {g=true, r=true}
   B=ScalarParameter() {defaultValue=0.15, parameterName=DistortStrengh}
  ) {constB=0.1}
  n27 = Panner(Coordinate=Add(A=Add(A=@n7, B=@n20), B=@n26), Time=@n24) {speedY=0.1}
  n28 = Add(A=Add(A=TextureCoordinate() {vTiling=0.75}, B=@n26), B=Constant2Vector() {g=0.25})
  n29 = Panner(Coordinate=LinearInterpolate(A=@n28, B=Multiply(A=@n8, B=@n28), Alpha=ScalarParameter() {parameterName=UseTextureTiling?}), Time=@n24) {speedX=0.15, speedY=0.5}
  n30 = Add(A=@n26, B=TextureCoordinate() {vTiling=1.1})
  n31 = Panner(Coordinate=LinearInterpolate(A=@n30, B=Multiply(A=@n14, B=@n30), Alpha=ScalarParameter() {parameterName=UseTextureTiling?})) {speedX=0.1, speedY=1}
  n32 = ComponentMask(None=TextureCoordinate() {vTiling=1.1}) {g=true}
  n33 = Constant() {r=0.1}
  n34 = Constant() {r=0.15}
  n35 = Multiply(
   A=Multiply(
    A=TextureSample(UVs=@n27) {texture=/Game/Effects/Fort_Effects/Textures/Tile/T_Inky_Smoke_Tile2}.R
    B=TextureSample(UVs=@n29) {texture=&T_Inky_Smoke_Tile2}.G
   )
   B=Multiply(
    A=TextureSample(UVs=@n31) {texture=/Game/Global/Textures/Default/Noise/T_FireNoise_02}.G
    B=MaterialFunctionCall(Alpha=@n32, Min=@n33, Max=@n34) {materialFunction=&SmoothStep}.Result
   )
  )
  n36 = QualitySwitch(Default=ShadingPathSwitch(Default=Multiply(A=@n16, B=@n35), Mobile=@n35) {desc=Migrated from FeatureLevelSwitch}, Low=@n35)
  n37 = ComponentMask(None=TextureCoordinate()) {g=true}
  n38 = OneMinus(None=@n37)
  n39 = Multiply(A=Multiply(A=ShadingPathSwitch(Default=@n36, Forward=@n35), B=@n37), B=@n38)
  n40 = Multiply(A=Multiply(A=Multiply(A=@n6, B=@n39), B=VertexColor().R), B=ScalarParameter() {defaultValue=0.5, parameterName=Opacity})
  n41 = ScalarParameter() {defaultValue=250, parameterName=FadeDistance}
  n42 = ShadingPathSwitch(Default=Multiply(A=@n40, B=DepthFade(FadeDistance=@n41)), Mobile=@n40) {desc=Migrated from FeatureLevelSwitch}
  Multiply(
   A=ShadingPathSwitch(Default=QualitySwitch(Default=@n42, Low=@n40), Forward=@n40)
   B=MaterialFunctionCall!(
    NearFadeLength=ScalarParameter() {defaultValue=21000, group=CloseFarFader, parameterName="NearFadeLength (S)"}
    NearFadeOffset=ScalarParameter() {defaultValue=20000, group=CloseFarFader, parameterName="NearFadeOffset (S)"}
    NearFadeCurve
    NearFadeStartingScale
    NearFadeEndingScale
    FarFadeLength=ScalarParameter() {defaultValue=25000, group=CloseFarFader, parameterName="FarFadeLength (S)"}
    FarFadeOffset=ScalarParameter() {defaultValue=125000, group=CloseFarFader, parameterName="FarFadeOffset (S)"}
    FarFadeCurve
    FarFadeEndingScale
    Opacity Near Distance Bias=ScalarParameter() {group=CloseFarFader, parameterName="Opacity Near Distance Bias Opacity (S)"}
    Opacity Far Distance Bias=ScalarParameter() {defaultValue=-2, group=CloseFarFader, parameterName="Opacity Far Distance Bias (S)"}
   ) {materialFunction=/Game/Effects/Fort_Effects/Materials/_Functions/MF_SmoothLod}.Opacity
  )
  n43 = ScalarParameter() {defaultValue=0.1, parameterName=HueShift_Max_01}
  n44 = VertexColor()
  n45 = ScalarParameter() {defaultValue=0.25, parameterName=HueShift_Max_02}
  n46 = VertexColor()
  n47 = LinearInterpolate(
   A=MaterialFunctionCall(
    Hue Shift Percentage=LinearInterpolate(B=@n43, Alpha=@n44.R) {constB=0.1}
    Texture=VectorParameter() {parameterName=FireColor_Low, defaultValue=<rgba 0 1 0.05 0>}.RGB
   ) {materialFunction=&HueShift}.Result
   B=MaterialFunctionCall(
    Hue Shift Percentage=LinearInterpolate(B=@n45, Alpha=@n46.R) {constB=0.25}
    Texture=VectorParameter() {parameterName=FireColor_High, defaultValue=<rgba 0 1 0 0>}.RGB
   ) {materialFunction=&HueShift}.Result
   Alpha=@n38
  )
  n48 = MaterialFunctionCall(
   LDR Input=@n47
   HDR Dynamic Range=ScalarParameter() {defaultValue=1, parameterName=Dynamic Range}
   HDR Intensity Multiplier=ScalarParameter() {defaultValue=10, parameterName=HDR Intensity}
   Derive HDR Power=ScalarParameter() {defaultValue=1, parameterName="Derive HDR Power (S)"}
  ) {materialFunction=&DeriveHDRfromLDR}
  n49 = Multiply(A=ScalarParameter() {defaultValue=10, parameterName=Mobile Color Intensity}, B=@n47)
  ShadingPathSwitch(Default=QualitySwitch(Default=ShadingPathSwitch(Default=@n48.RemappedRange, Mobile=@n49) {desc=Migrated from FeatureLevelSwitch}, Low=@n49))
  n50 = ScalarParameter() {defaultValue=250, parameterName=WidthBase}
  n51 = Multiply(B=VertexColor().G) {constA=25}
  n52 = MaterialFunctionCall(
   NearFadeLength=ScalarParameter() {defaultValue=21000, group=CloseFarFader, parameterName="NearFadeLength (S)"}
   NearFadeOffset=ScalarParameter() {defaultValue=20000, group=CloseFarFader, parameterName="NearFadeOffset (S)"}
   FarFadeLength=ScalarParameter() {defaultValue=25000, group=CloseFarFader, parameterName="FarFadeLength (S)"}
   FarFadeOffset=ScalarParameter() {defaultValue=125000, group=CloseFarFader, parameterName="FarFadeOffset (S)"}
   Opacity Near Distance Bias=ScalarParameter() {group=CloseFarFader, parameterName="Opacity Near Distance Bias (S)"}
   Opacity Far Distance Bias=ScalarParameter() {defaultValue=-2, group=CloseFarFader, parameterName="Opacity Far Distance Bias (S)"}
  ) {materialFunction=&MF_SmoothLod}
  n53 = ScalarParameter() {defaultValue=500, parameterName=WidthTip}
  n54 = Multiply(B=VertexColor().G) {constA=100}
  n55 = MaterialFunctionCall(WidthBase=Multiply(A=Add(A=@n50, B=@n51), B=@n52.Opacity), WidthTip=Multiply(A=Add(A=@n53, B=@n54), B=@n52.Opacity)) {materialFunction=&SplineThicken}
  n56 = Constant3Vector()
  n57 = VectorParameter() {group=WPO, parameterName=TextureWPO_UVtiling, defaultValue=<rgba 1 2 0 0>}
  n58 = Panner(Coordinate=Multiply(A=TextureCoordinate(), B=ComponentMask(None=@n57.RGB) {g=true, r=true})) {speedY=0.15}
  n59 = TextureSample(UVs=@n58) {samplerType=SAMPLERTYPE_Normal, texture=&T_Weap_Distort_N_02}
  n60 = ComponentMask(None=@n59.RGB) {r=true}
  n61 = ScalarParameter() {defaultValue=25, group=WPO, parameterName=WPO_HeightX}
  n62 = ComponentMask(None=@n59.RGB) {g=true}
  n63 = ScalarParameter() {group=WPO, parameterName=WPO_HeightY}
  n64 = AppendVector(
   A=AppendVector(A=Multiply(A=@n60, B=@n61) {constB=0.5}, B=Multiply(A=@n62, B=@n63) {constB=0.5})
   B=Multiply(
    A=TextureSample(UVs=@n58) {texture=&T_Smooth_tile}.G
    B=ScalarParameter() {group=WPO, parameterName=WPO_HeightZ}
   ) {constB=0.6}
  )
  n65 = Constant() {r=0.5}
  n66 = SphereMask(
   A=MaterialFunctionCall() {materialFunction=&LinearGradient}.VGradient
   B=@n65
   Radius=@n65
  ) {hardnessPercent=0}
  n67 = Multiply(
   A=Transform(None=LinearInterpolate(A=@n56, B=@n64, Alpha=@n66))
   B=LinearInterpolate(
    A=ScalarParameter() {group=WPO, parameterName=WPO_Wave_Intensity_Close}
    B=ScalarParameter() {defaultValue=100, group=WPO, parameterName=WPO_Wave_Intenstiy}
    Alpha=@n52.Opacity
   )
  ) {constB=80}
  n68 = Constant3Vector() {desc=Disable WPO anim in Low, constant=<rgba 0 0 20 0>}
  n69 = ShadingPathSwitch(Default=@n67, Deferred=@n67, Mobile=@n68) {desc=Migrated from FeatureLevelSwitch}
  n70 = Add(A=@n55.WorldPositionOffset, B=QualitySwitch(Default=@n69, Low=@n68, High=@n69, Medium=@n69))
  ShadingPathSwitch(
   Default=QualitySwitch(
    Default=ShadingPathSwitch(Default=@n70, Mobile=@n55.WorldPositionOffset) {desc=Migrated from FeatureLevelSwitch}
    Low=@n55.WorldPositionOffset
   )
   Forward=@n55.WorldPositionOffset
  )
 ]
}
