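# Translucent unlit master material for BlackMonday light effects, with static switches selecting light cone, flare or end-symbol looks.
# Opacity combines view-angle or Fresnel, vertical, distance and depth fades with flare or symbol masks and panning, optionally triplanar, fog; glow color follows time of day.
# World position offset widens cones along normals or builds billboard and distant-flare sprites.
# The countdown collection's DisableLightcone parameter collapses vertices to world origin, hiding the effect.
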Material /Game/Effects/Fort_Effects/Materials/Environment/BlackMonday/M_BlackMonday_LightCone_Master
graph {
 properties {
  bUsedWithInstancedStaticMeshes=true
  bUsedWithMeshParticles=true
  blendMode=BLEND_Translucent
  pixelDepthOffsetMode=PDOM_Legacy
  shadingModel=MSM_Unlit
  twoSided=true
  usedShadingModels=Unlit
 }
 nodes [
  Constant() {r=1}
  MaterialFunctionCall() {materialFunction=&Fresnel_Function}
  n1 = ScalarParameter() {defaultValue=1, group=Color Adjustments, parameterName=Morning Multiplier}
  n2 = VectorParameter() {parameterName=GlowColor, defaultValue=<rgba 1 0.5328 0.124 1>}
  n3 = ScalarParameter() {defaultValue=1, group=Color Adjustments, parameterName=Day Multiplier}
  n4 = ScalarParameter() {defaultValue=1, group=Color Adjustments, parameterName=Evening Multiplier}
  n5 = ScalarParameter() {defaultValue=1, group=Color Adjustments, parameterName=Night Multiplier}
  MaterialFunctionCall(Emissive=MaterialFunctionCall(
    Color_Morning=Multiply(A=@n1, B=@n2.RGB) {constB=250}
    Color_Day=Multiply(A=@n2.RGB, B=@n3) {constB=250}
    Color_Evening=Multiply(A=@n4, B=@n2.RGB) {constB=250}
    Color_Night=Multiply(A=@n5, B=@n2.RGB) {constB=250}
   ) {materialFunction=&Color_Time_of_Day}.Result) {materialFunction=&MF_FX_EmissiveAdjustments}
  n6 = ScalarParameter() {defaultValue=4, group=Fresnel, parameterName=Base Fresnel Exponent}
  n7 = MaterialFunctionCall("FadeSensitivity (optional)"=@n6) {materialFunction=&FadeBasedOnViewAngle}
  n8 = OneMinus(None=Fresnel(ExponentIn=@n6))
  n9 = MaterialFunctionCall(
   Position=WorldPosition().XYZ
   Fade_Start_Distance=ScalarParameter() {defaultValue=250, group=Fade, parameterName="Fade_Start_Distance (S)"}
   Fade_End_Distance=ScalarParameter() {defaultValue=100, group=Fade, parameterName="Fade_End_Distance (S)"}
  ) {materialFunction=&MF_Distance_SyncedFade_FX}
  n10 = Multiply(
   A=Power(
    Base=StaticSwitchParameter(True=@n7.Result, False=@n8) {parameterName=UseViewAngleFade}
    Exp=ScalarParameter() {defaultValue=12, group=Fresnel, parameterName=Base Fresnel Power}
   )
   B=@n9.Result
  )
  n11 = ComponentMask(None=TextureCoordinate()) {g=true}
  n12 = ScalarParameter() {defaultValue=1, group=Fade, parameterName=Vertical Fade Power}
  n13 = Saturate(None=Multiply(A=Power(Base=@n11, Exp=@n12), B=ScalarParameter() {defaultValue=2, group=Fade, parameterName=Vertical Fade Mult}))
  n14 = MaterialFunctionCall(Texture Scale=ScalarParameter() {defaultValue=1.2, parameterName=CutoutTextureScale}) {materialFunction=&ScaleUVsByCenter}
  n15 = TextureSampleParameter2D!(UVs=@n14.UVs, Bias, Apply View MipBias) {group=Flare, mipValueMode=TMVM_MipBias, parameterName=Flare Texture, texture=/Game/Effects/Fort_Effects/Textures/Flares/T_Flare_Soft}
  n16 = TextureSampleParameter2D(UVs=@n14.UVs, Bias=ScalarParameter() {defaultValue=-1, parameterName=Symbol LOD Bias}) {
   group=End Symbol
   mipValueMode=TMVM_MipBias
   parameterName=Symbol Texture
   samplerSource=SSM_Clamp_WorldGroupSettings
   texture=/Game/Effects/Fort_Effects/Textures/Environments/BlackMonday/Strawberry/T_BlackMonday_Strawberry_Small
  }
  n17 = Multiply(A=@n14.0-1 mask, B=ChannelMaskParameter(None=@n16.RGBA) {parameterName=SymbolTextureChannel}.R)
  n18 = TextureCoordinate() {coordinateIndex=1}
  n19 = Constant() {r=0.5}
  n20 = ScalarParameter() {defaultValue=0.5, group=End Symbol, parameterName=Radius}
  n21 = ScalarParameter() {defaultValue=0.5, group=End Symbol, parameterName=Hardness}
  n22 = Multiply(
   A=Multiply(A=OneMinus(None=@n17)) {constB=0.5}
   B=Power(
    Base=SphereMask(A=@n18, B=@n19, Radius=@n20, Hardness=@n21)
    Exp=ScalarParameter() {defaultValue=2, group=End Symbol, parameterName=SphereMask Power}
   )
  )
  n23 = MaterialFunctionCall("FadeSensitivity (optional)"=ScalarParameter() {group=Fade, parameterName=Angle Fade Sensitivity}) {materialFunction=&FadeBasedOnViewAngle}
  n24 = StaticSwitchParameter(
   True=StaticComponentMaskParameter(None=@n15.RGB) {defaultG=true, group=Flare, parameterName=Flare Texture Channel}
   False=StaticSwitchParameter(True=Multiply(A=@n22, B=@n23.Result), False=@n22) {parameterName=Fade Based On Angle?}
  ) {parameterName=IsFlare?}
  n25 = Multiply(
   A=DepthFade(FadeDistance=ScalarParameter() {parameterName=DepthFade}) {fadeDistanceDefault=25}
   B=StaticSwitchParameter(True=Multiply(A=@n10, B=@n13), False=Multiply(A=@n9.Result, B=@n24)) {parameterName=IsLightCone?}
  )
  n26 = ScalarParameter() {defaultValue=1, parameterName=Final Power}
  n27 = Multiply(A=Clamp(None=Power(Base=@n25, Exp=@n26)), B=ScalarParameter() {defaultValue=0.1, parameterName=Final Multiplier})
  n28 = AppendVector(
   A=ScalarParameter() {defaultValue=0.005, group=Fog, parameterName=Fog Pan Speed X}
   B=ScalarParameter() {group=Fog, parameterName=Fog Pan Speed Y}
  )
  n29 = TextureObjectParameter() {group=Fog, parameterName=Fog Texture, texture=&T_Smooth_tile}
  n30 = VectorParameter() {group=Fog, parameterName=FogTextureTiling, defaultValue=<rgba 2 2 2 1>}
  n31 = Time()
  n32 = Add(A=Multiply(A=TextureCoordinate(), B=ComponentMask(None=@n30.RGB) {g=true, r=true}), B=Frac(None=Multiply(A=@n31, B=@n28)))
  n33 = LinearInterpolate(
   A=StaticSwitchParameter(
    True=MaterialFunctionCall!(CameraVector, Time Scale, Panner Speed=@n28, Texture=@n29, Tiling=@n30.RGB, Offset, AxisFadeContrast) {materialFunction=/Game/Weapons/WeaponSkins/MaterialFunction/TriplanarCamerVector_Panning}.Result
    False=TextureSample(UVs=@n32) {texture=&T_Smooth_tile}.G
   ) {defaultValue=true, group=Fog, parameterName=UseTriPlanarFog?}
   Alpha=ScalarParameter() {defaultValue=0.5, group=Fog, parameterName=Fade Fog}
  )
  Multiply(
   A=LinearInterpolate(A=ParticleColor().A, B=ScalarParameter() {defaultValue=1, parameterName=Alpha}, Alpha=ScalarParameter() {parameterName=Use_Mesh_Alpha})
   B=Clamp(None=Multiply(A=@n27, B=@n33))
  )
  n34 = TextureCoordinate()
  n35 = Multiply(
   A=VertexNormalWS()
   B=LinearInterpolate(A=ScalarParameter() {parameterName=Scale 1}, B=ScalarParameter() {parameterName=Scale 2}, Alpha=ComponentMask(None=@n34) {g=true})
  )
  n36 = Normalize(VectorInput=CameraPositionWS())
  n37 = VectorParameter() {group=Secondary Flare, parameterName=LightViewDirectionAxis, defaultValue=<rgba 1 0 0 1>}
  n38 = Normalize(VectorInput=StaticSwitchParameter(
    True=Transform(None=@n37.RGB) {transformSourceType=TRANSFORMSOURCE_ParticleWorld}
    False=Transform(None=@n37.RGB) {transformSourceType=TRANSFORMSOURCE_Local}
   ) {parameterName=IsMeshParticle?})
  n39 = Clamp(None=Power(Base=DotProduct(A=@n36, B=@n38), Exp=ScalarParameter() {defaultValue=4.5, group=Secondary Flare, parameterName=Camera Direction Fade Power}) {constExponent=3})
  n40 = AppendVector(
   A=ScalarParameter() {defaultValue=100, group=Flare, parameterName=Sprite Scale X}
   B=ScalarParameter() {defaultValue=100, group=Flare, parameterName=Sprite Scale Y}
  )
  n41 = ScalarParameter() {defaultValue=10, group=Secondary Flare, parameterName=Secondary Flare Scale Multiplier}
  n42 = CameraPositionWS()
  n43 = MaterialFunctionCall() {materialFunction=&ObjectPivotPoint}
  n44 = Divide(
   A=Subtract(A=Distance(A=@n42, B=@n43.Object Pivot Location), B=ScalarParameter() {group=Secondary Flare, parameterName=Secondary Flare Scale Offset})
   B=ScalarParameter() {group=Secondary Flare, parameterName=Secondary Flare Scale Length}
  )
  n45 = Multiply(A=LinearInterpolate(A=@n40, B=Multiply(A=@n40, B=@n41) {constB=7}, Alpha=Saturate(None=@n44)), B=@n40)
  n46 = MaterialFunctionCall!(
   Center Location=WorldPosition().XYZ
   0-1 UVs
   XY scale=StaticSwitchParameter(True=Multiply(A=@n39, B=@n45), False=@n40) {parameterName=IsDistantFlare}
   OverridePivot
   "Normals (Optional)"
  ) {materialFunction=&Sprite}
  n47 = LinearInterpolate(B=ScalarParameter() {defaultValue=500, group=Secondary Flare, parameterName=Camera Offset Amount}, Alpha=@n39)
  n48 = StaticSwitchParameter(
   True=Add(
    A=@n46.World position offset
    B=MaterialFunctionCall(Offset Amount=@n47) {materialFunction=/Engine/Functions/Engine_MaterialFunctions02/WorldPositionOffset/CameraOffset}.Camera Offset
   )
   False=StaticSwitchParameter(True=@n46.World position offset, False=Constant()) {defaultValue=true, parameterName=IsBillboard?}
  ) {parameterName=IsDistantFlare}
  n49 = WorldPosition()
  n50 = CollectionParameter() {collection=/Game/Athena/Events/Generic/CountdownTimer/Materials/MPC_Countdown, parameterName=DisableLightcone}
  Reroute(None=LinearInterpolate(
    A=StaticSwitchParameter(True=@n35, False=@n48) {parameterName=IsLightCone?}
    B=Multiply(A=@n49.XYZ) {constB=-1}
    Alpha=ComponentMask(None=@n50) {r=true}
   ))
  ScalarParameter() {defaultValue=1, parameterName=GONE}
  TransformPosition(None=@n37.RGB) {transformSourceType=TRANSFORMPOSSOURCE_Particle, transformType=TRANSFORMPOSSOURCE_World}
 ]
}
